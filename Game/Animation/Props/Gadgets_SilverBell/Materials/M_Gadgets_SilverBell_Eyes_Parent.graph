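# Parent material for the SilverBell gadget prop that procedurally draws animatable eyes on the standard 2023 character PreFX/PostFX shader, with optional Slurp.
# Sphere masks in UV1 form eyelid rings and clipped eyeballs, split left/right by UV tile, with R_Eye and L_Eye interpolating lid radius.
# The combined mask blends eye color, emissive, metallic, specular and roughness over the base.
# Lid masks also become a height-derived, flattened normal blended into the surface.
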Material /Game/Animation/Props/Gadgets_SilverBell/Materials/M_Gadgets_SilverBell_Eyes_Parent
graph {
 properties {
  bUseMaterialAttributes=true
  bUsedWithClothing=true
  bUsedWithMeshDeformer=true
  bUsedWithNanite=true
  bUsedWithSkeletalMesh=true
  pixelDepthOffsetMode=PDOM_Legacy
  shadingModel=MSM_FromMaterialExpression
  usedShadingModels=Default Lit | Subsurface | Clear Coat | Subsurface Profile | Cloth
 }
 nodes [
  n1 = MaterialFunctionCall() {materialFunction=&MF_CharacterParent_2023_PreFX}
  n2 = StaticComponentMaskParameter(None=@n1.Tech Art Mask) {defaultR=true, group=Skin FX, parameterName=Slurp Channel}
  n3 = ComponentMask(None=@n1.M) {r=true}
  n4 = StaticComponentMaskParameter(None=@n1.Tech Art Mask) {defaultG=true, group=Skin FX, parameterName=Slurp Bubbles Channel}
  n5 = StaticComponentMaskParameter(None=@n1.Tech Art Mask) {defaultB=true, group=Skin FX, parameterName=Slurp Normal Distortion Channel}
  n6 = GetMaterialAttributes(None=StaticSwitchParameter(
    True=MaterialFunctionCall!(
     Alpha=@n2
     AO=@n3
     BubblesMask=@n4
     SlurpNormalDistortionMask=@n5
     Pre Skinned Position=@n1.Pre skinned Local Position
     Material=@n1.Pre skinned Local Position
    ) {materialFunction=/Game/Weapons/WeaponSkins/MaterialFunction/MF_Slurp}.Result
    False=@n1.Result
   ) {group=Skin FX, parameterName=UseSlurp}) {attributeGetTypes [
    #1
    #2
    #3
    #4
    #5
    #6
   ]}
  n7 = NamedRerouteUsage()
  n8 = VectorParameter() {parameterName=EyeColor, defaultValue=<rgba 0.015 0.015 0.015 1>}
  n9 = NamedRerouteUsage()
  n10 = SetMaterialAttributes!(
   MaterialAttributes=@n6.MaterialAttributes
   Base Color=LinearInterpolate(A=Multiply(A=@n7) {constB=0.4}, B=@n8.RGB, Alpha=@n9) {constB=0}
   Emissive Color=LinearInterpolate(A=Multiply(A=@n6.Emissive Color) {constB=0.5}, B=@n8.RGB, Alpha=@n9) {constB=0}
   Metallic=LinearInterpolate(A=@n6.Metallic, B=ScalarParameter() {group=SilverBellMarineProp, parameterName=EyeMetallic}, Alpha=@n9) {constB=0}
   Specular=LinearInterpolate(A=@n6.Specular, B=ScalarParameter() {defaultValue=0.35, group=SilverBellMarineProp, parameterName=EyeSpecular}, Alpha=@n9) {constB=0}
   Roughness=LinearInterpolate(A=@n6.Roughness, B=ScalarParameter() {defaultValue=0.5, group=SilverBellMarineProp, parameterName=EyeRoughness}, Alpha=@n9) {constB=0}
   Normal=MaterialFunctionCall(BaseNormal=NamedRerouteUsage(), AdditionalNormal=NamedRerouteUsage()) {materialFunction=&BlendAngleCorrectedNormals}.Result
  ) {attributeSetTypes [
    #1
    #2
    #3
    #4
    #5
    #6
   ]}
  MaterialFunctionCall(Material=@n10, M=@n1.M, Pre Skinned Local Position=@n1.M, Distance Blend=@n1.M, Fresnel=@n1.M) {materialFunction=&MF_CharacterParent_2023_PostFX}
  n11 = Frac(None=TextureCoordinate() {coordinateIndex=1})
  n12 = ScalarParameter() {defaultValue=1.8, group=SilverBellMarineProp, parameterName=Y}
  n13 = MaterialFunctionCall(X=Constant() {r=0.63}, Y=@n12) {materialFunction=&MakeFloat2}
  n14 = ScalarParameter() {defaultValue=2.75, group=SilverBellMarineProp, parameterName=Eyelid_Max}
  n15 = ScalarParameter() {defaultValue=2.3, group=SilverBellMarineProp, parameterName=Eyelid_Min}
  n16 = LinearInterpolate(A=@n14, B=@n15, Alpha=Clamp(None=ScalarParameter() {group=SilverBellMarineProp, parameterName=R_Eye})) {constA=1.9, constB=1.4}
  n17 = SphereMask(A=@n11, B=@n13.Result, Radius=@n16) {attenuationRadius=0.6, hardnessPercent=0}
  n18 = MaterialFunctionCall(X=Constant() {r=0.6}, Y=@n12) {materialFunction=&MakeFloat2}
  n19 = Add(A=ScalarParameter() {defaultValue=0.03, group=SilverBellMarineProp, parameterName=EyelidThickness}) {constB=2}
  n20 = Subtract(A=@n17, B=SphereMask(A=@n11, B=@n18.Result, Radius=Divide(A=@n16, B=@n19) {constB=3}) {attenuationRadius=0.6, hardnessPercent=0})
  n21 = NamedRerouteUsage()
  n22 = Multiply(A=SmoothStep(Value=@n20) {constMax=0.5, constMin=0.49}, B=@n21)
  n23 = LinearInterpolate(A=@n14, B=@n15, Alpha=Clamp(None=ScalarParameter() {group=SilverBellMarineProp, parameterName=L_Eye})) {constA=2, constB=2.4}
  n24 = SphereMask(A=@n11, B=@n13.Result, Radius=@n23) {attenuationRadius=0.6, hardnessPercent=0}
  n25 = Subtract(A=@n24, B=SphereMask(A=@n11, B=@n18.Result, Radius=Divide(A=@n23, B=@n19) {constB=3}) {attenuationRadius=0.6, hardnessPercent=0})
  n26 = NamedRerouteUsage()
  n27 = Multiply(A=SmoothStep(Value=@n25) {constMax=0.5, constMin=0.49}, B=@n26)
  n28 = Multiply(A=Add(
    A=NamedRerouteDeclaration(None=@n22) {name=R_Eyelid, nodeColor=<rgba 0 0.6953 1 1>}
    B=NamedRerouteDeclaration(None=@n27) {name=L_Eyelid, nodeColor=<rgba 0 0.6953 1 1>}
   )) {constB=10}
  n29 = Step(X=@n17) {constX=0.5, constY=0.5}
  n30 = Step(X=@n24) {constX=0.5, constY=0.5}
  n31 = NamedRerouteUsage()
  n32 = Multiply(A=Frac(None=TextureCoordinate() {coordinateIndex=1}), B=Constant2Vector() {g=0.8, r=1})
  n33 = ComponentMask(None=VectorParameter() {group=SilverBellMarineProp, parameterName=EyeballPosition, defaultValue=<rgba 0.52 0.45 0 1>}.RGB) {g=true, r=true}
  n34 = Multiply(A=Multiply(A=LinearInterpolate(A=@n29, B=@n30, Alpha=@n31), B=SphereMask(A=@n32, B=@n33) {attenuationRadius=0.11})) {constB=10}
  NamedRerouteDeclaration(None=Max(A=Saturate(None=@n28), B=Saturate(None=@n34))) {name=EyeMask, nodeColor=<rgba 1 0.75 0 1>}
  n35 = SmoothStep(Value=@n20) {constMax=0.501, constMin=0.489}
  n36 = SmoothStep(Value=@n25) {constMax=0.501, constMin=0.489}
  n37 = MaterialFunctionCall!(
   Height=Add(A=Multiply(A=@n35, B=@n21), B=Multiply(A=@n36, B=@n26))
   Absolute World Position
   World Space Vertex Normals=Transform(None=NamedRerouteUsage())
  ) {materialFunction=/Engine/Functions/Engine_MaterialFunctions03/Procedurals/HeightToNormalSmooth}
  NamedRerouteDeclaration(None=MaterialFunctionCall(
    Normal=Transform(None=@n37.WorldSpaceNormal) {transformSourceType=TRANSFORMSOURCE_World, transformType=TRANSFORM_Tangent}
    Flatness=ScalarParameter() {defaultValue=0.5, group=SilverBellMarineProp, parameterName=NormalFlatness}
   ) {materialFunction=&FlattenNormal}.Result) {name=EyelidNormal, nodeColor=<rgba 0 1 0.7422 1>}
  n38 = Floor(None=Add(A=TextureCoordinate() {coordinateIndex=1}) {constB=0})
  n39 = MaterialFunctionCall(Float2=@n38) {materialFunction=&BreakOutFloat2Components}
  n40 = Max(A=Add(A=@n39.R) {constB=-1}, B=@n39.G) {constB=0}
  n41 = MaterialFunctionCall(Float2=Multiply(A=@n38) {constB=-1}) {materialFunction=&BreakOutFloat2Components}
  n42 = Max(A=@n41.R, B=@n41.R) {constB=0}
  NamedRerouteDeclaration(None=OneMinus(None=Max(A=@n40, B=@n42) {constB=0})) {name=Mask_EyesUDIM, nodeColor=<rgba 1 0.2578 0 1>}
  n43 = MaterialFunctionCall(Float2=TextureCoordinate() {coordinateIndex=1}) {materialFunction=&BreakOutFloat2Components}
  n44 = Floor(None=@n43.R) {desc=LeftEye}
  n45 = Min(A=Ceil(None=@n43.G), B=NamedRerouteUsage())
  NamedRerouteDeclaration(None=Multiply(A=OneMinus(None=@n44), B=NamedRerouteDeclaration(None=@n45) {name=Mask_Eyes, nodeColor=<rgba 1 0 0.8906 1>}) {desc=RightEye}) {name=Mask_EyeRight, nodeColor=<rgba 1 0 0.0152 1>}
  NamedRerouteDeclaration(None=@n44) {name=Mask_EyeLeft, nodeColor=<rgba 0.0293 0.3125 0.0353 1>}
  NamedRerouteDeclaration(None=GetMaterialAttributes(None=@n1.Result) {attributeGetTypes [#1]}.Base Color) {name=BaseColor_IN, nodeColor=<rgba 0 0.7891 1 1>}
  NamedRerouteDeclaration(None=@n6.Normal) {name=Normal_IN, nodeColor=<rgba 1 0.5391 0 1>}
  Reroute()
  Step(Y=MaterialFunctionCall(Speed=Constant() {r=1.5}) {materialFunction=&DebugTimeSine}.Result) {constX=0.5}
 ]
}
